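; Construction script for the BP_Academy_MSLRoofTower_A academy roof tower prop, setting up mesh mobility and time-of-day lighting.
; Unless the mobility override variable is set, it loops over all static mesh components and sets each to Static.
; Inside that loop, when the TOD component is enabled, it calls RunConstruct on the TOD lighting component with the test day phase, so the call repeats per mesh and is skipped under the override.
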(fn ConstructionScript
  (if (not (Variables|Mesh|GetOverrideStaticMeshMobility_0))
    (for _array_element (Actor|GetComponentsByClass "/Script/Engine.StaticMeshComponent")
      (Transformation|SetMobility _array_element "Static")
      (if (Variables|Default|GetUse%20TOD%20Component)
        (Class|TODLightingComponent|RunConstruct (Variables|Default|GetTODLightingComponent) (Variables|Default|GetTest%20Day%20Phase))))))
